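# MetaSound Source that plays a single wave asset as a 5.1 surround one-shot, with the Sound input left empty by default.
# On Play triggers a 5.1 Wave Player, whose six channels go straight to the matching outputs with no further processing.
# The player's On Finished trigger drives the one-shot OnFinished output to signal the end of the source.
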MetaSoundSource /MetaSoundWizard/MetaSounds/One-shots/Single/MSS_Oneshot_Single_5dot1
inputs [
 Sound {type=WaveAsset}
 UE.Source.OnPlay {type=Trigger}
]
interfaces [UE.OutputFormat.5dot1:1.0, UE.Source.OneShot:1.0, UE.Source:1.1]
outputs [
 "UE.OutputFormat.5dot1.Audio:0" {type=Audio}
 "UE.OutputFormat.5dot1.Audio:1" {type=Audio}
 "UE.OutputFormat.5dot1.Audio:2" {type=Audio}
 "UE.OutputFormat.5dot1.Audio:3" {type=Audio}
 "UE.OutputFormat.5dot1.Audio:4" {type=Audio}
 "UE.OutputFormat.5dot1.Audio:5" {type=Audio}
 UE.Source.OneShot.OnFinished {type=Trigger}
]
pages [Default {
  input_defaults {Sound=null, UE.Source.OnPlay=null}
  nodes [WavePlayer_0 = UE.Wave Player:5dot1(Wave Asset=^Sound, Play=^UE.Source.OnPlay)]
  links [
   {dst="@UE.Source.OneShot.OnFinished", src=WavePlayer_0.On Finished}
   {dst="@UE.OutputFormat.5dot1.Audio:2", src=WavePlayer_0.Out Front Center}
   {dst="@UE.OutputFormat.5dot1.Audio:0", src=WavePlayer_0.Out Front Left}
   {dst="@UE.OutputFormat.5dot1.Audio:1", src=WavePlayer_0.Out Front Right}
   {dst="@UE.OutputFormat.5dot1.Audio:3", src=WavePlayer_0.Out Low Frequency}
   {dst="@UE.OutputFormat.5dot1.Audio:4", src=WavePlayer_0.Out Side Left}
   {dst="@UE.OutputFormat.5dot1.Audio:5", src=WavePlayer_0.Out Side Right}
  ]
 }]
properties {
 assetImportData=&self:AssetImportData
 numChannels=6
 outputFormat=FiveDotOne
}
