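# MetaSound Source template for a single mono one-shot sound.
# An exposed Sound wave asset input feeds a mono Wave Player that starts on the source's OnPlay trigger.
# The player's audio goes straight to the mono output, and its On Finished trigger drives the one-shot OnFinished output so the source ends when the wave does.
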MetaSoundSource /MetaSoundWizard/MetaSounds/One-shots/Single/MSS_Oneshot_Single_Mono
inputs [
 Sound {type=WaveAsset}
 UE.Source.OnPlay {type=Trigger}
]
interfaces [UE.OutputFormat.Mono:1.0, UE.Source.OneShot:1.0, UE.Source:1.1]
outputs [
 "UE.OutputFormat.Mono.Audio:0" {type=Audio}
 UE.Source.OneShot.OnFinished {type=Trigger}
]
pages [Default {
  input_defaults {Sound=null, UE.Source.OnPlay=null}
  nodes [WavePlayer_0 = UE.Wave Player:Mono(Wave Asset=^Sound, Play=^UE.Source.OnPlay)]
  links [{dst="@UE.Source.OneShot.OnFinished", src=WavePlayer_0.On Finished}, {dst="@UE.OutputFormat.Mono.Audio:0", src=WavePlayer_0.Out Mono}]
 }]
properties {assetImportData=&self:AssetImportData}
